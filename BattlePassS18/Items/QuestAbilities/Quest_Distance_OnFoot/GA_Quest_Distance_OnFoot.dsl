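; Quest ability that tracks distance the owning player travels on foot for Battle Pass S18 quests.
; A looping timer, with a randomized interval for perf, adds floored XY-plane meters travelled while the pawn moves on ground without restricted tags.
; Otherwise it resets the stored locations, so airborne, vehicle, carried or burrowing movement is not counted.
; Totals go to the quest manager as Athena.Quests.Distance.OnFoot stat events after exceeding 100 meters, or on death.
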(event Ability|EventActivateAbility
  (bind _owningplayerpawn (Variables|GAVars|Data|GetOwningPlayerPawn))
  ;[1 1 1] Set the player's current and previous location to match
  (|CallParentFunction)
  (Utilities|IsValid _owningplayerpawn
    (:Is%20Valid
      (bind _output_get (Variables|Default|SetPreviousLocation (Transformation|GetActorLocation _owningplayerpawn)))
      (Variables|Default|SetCurrentLocation _output_get)
      ;[1 1 1] Tick at close random intervals for perf, do not go under 1 second
      (EventDispatchers|BindEventtoOnDied _owningplayerpawn (EventDispatchers|CreateEvent :!event Custom|Died))
      (bind _output_get_2 (Variables|Default|SetInterval (Math|Random|RandomFloatinRange 1.7 2.5)))
      (Utilities|Time|SetTimerbyEvent (delegate Custom|SendDistanceUpdate) _output_get_2 true))))

(event Custom|SendDistanceUpdate
  (bind _owningplayerpawn (Variables|GAVars|Data|GetOwningPlayerPawn))
  (bind _returnvalue (Transformation|GetActorLocation _owningplayerpawn))
  (bind _owningplayerpawn_2 (Variables|GAVars|Data|GetOwningPlayerPawn))
  (bind (_previouslocation_x _previouslocation_y) (Math|Vector|BreakVector (Variables|GetPreviousLocation_X)))
  (bind (_currentlocation_x _currentlocation_y) (Math|Vector|BreakVector (Variables|GetCurrentLocation_X)))
  ;[1 1 1] Make sure player is moving on the ground and is not in a vehicle, being carried, or sand burrowing
  (Utilities|IsValid _owningplayerpawn_2
    (:Is%20Valid
      (if (AI|Components|NavMovement|IsMovingonGround (Pawn|GetMovementComponent _owningplayerpawn_2))
        (bind _returnvalue_2 (MissionHelpers|HasTags (GameplayTags|GetOwnedGameplayTags (Variables|GAVars|Data|GetOwningPlayerPawn) :!class /Script/GameplayTags.GameplayTagAssetInterface) (Variables|Default|GetT_RestrictedTags) false false))
        (if _returnvalue_2
          ;[1 1 1] Calculate distance travelled along XY plane
          (bind _output_get (Variables|Default|SetPreviousLocation _returnvalue))
          (Variables|Default|SetCurrentLocation _output_get) ; Reset prev and curr locations while player remains in a vehicle
          (else
            (Variables|Default|SetCurrentLocation _returnvalue)
            (Variables|Default|SetMetersTravelled (+ (Math|Float|Floor (/ (Math|Vector|VectorLength (Math|Vector|vector-vector (Math|Vector|MakeVector _currentlocation_x _currentlocation_y) (Math|Vector|MakeVector _previouslocation_x _previouslocation_y))) 1e+02)) (Variables|Default|GetMetersTravelled)))
            (Variables|Default|SetPreviousLocation (Variables|Default|GetCurrentLocation)) ; Set the current location to the previous location
            ;[1 1 1] Send quest progress if distance is > 0
            (if (> (Variables|Default|GetMetersTravelled) 100)
              (Quest|SendComplexCustomStatEvent (Variables|GAVars|Data|GetOwningPlayerQuestManager) 0 "(GameplayTags=((TagName=\"Athena.Quests.Distance.OnFoot\")))" "(GameplayTags=)" (Variables|Default|GetMetersTravelled))
              (Variables|Default|SetMetersTravelled 0))))
        (else
          (bind _output_get_2 (Variables|Default|SetPreviousLocation (Transformation|GetActorLocation (Variables|GAVars|Data|GetOwningPlayerPawn))))
          (Variables|Default|SetCurrentLocation _output_get_2) ; Reset prev and curr locations while player is not on the ground
        )))))

(event Ability|EventFailedToActivatePassiveAbility
  (params (ActorInfo :type Gameplay%20Ability%20Actor%20Info%20Structure)))

(event Custom|Died
  (params (DamagedActor :type Actor%20Object%20Reference) (Damage :type Float%20%28single-precision%29) (InstigatedBy :type Controller%20Object%20Reference) (DamageCauser :type Actor%20Object%20Reference) (HitLocation :type Vector) (FHitComponent :type Primitive%20Component%20Object%20Reference) (BoneName :type Name) (Momentum :type Vector))
  (CallFunction|QuestUpdate))

(event Custom|Quest%20Update
  ;[1 1 1] Send quest progress if distance is > 0
  (Quest|SendComplexCustomStatEvent (Variables|GAVars|Data|GetOwningPlayerQuestManager) 0 "(GameplayTags=((TagName=\"Athena.Quests.Distance.OnFoot\")))" "(GameplayTags=)" (Variables|Default|GetMetersTravelled))
  (Variables|Default|SetMetersTravelled 0))
